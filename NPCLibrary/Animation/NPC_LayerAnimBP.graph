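# Animation layer Blueprint that overrides an NPC's full body with idle, turn-in-place and conversation animations.
# Named anim input events set one-frame flags, cleared post-evaluation, driving conversation start, loop, give/receive item and finish states.
# The default state layers additive noise, yaw aim offset and turns over a base pose; base and conversation clips vary by gender.
# Both weapon bones are scaled near zero to hide held weapons.
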AnimBlueprint /NPCLibrary/Animation/NPC_LayerAnimBP
target_skeleton=&Fortnite_M_Avg_Player_Skeleton
graphs [
 AnimGraph {nodes [Output Pose() {layerGroup=None}]}
 CharacterSkeletalControlPostLegIK {nodes [
   Input Pose() {name=InPosePostLegIK}
   Output Pose() {layerGroup=PlayerOverride}
  ]}
 EventGraph {nodes [
   n1 = Event BlueprintProcessAnimInputEvent()
   > n2 = Branch(Condition=IsNamed(self=@n1.AnimInputEvent, InEventName=aie.npc.ConversationStart))
   > Set IsConversationStarting(IsConversationStarting:bool=true) -> (Output_Get:bool)
   n3 = Branch(execute=@n2.else, Condition=IsNamed(self=@n1.AnimInputEvent, InEventName=aie.npc.ConversationEnd))
   > Set IsConversationEnding(IsConversationEnding:bool=true) -> (Output_Get:bool)
   n4 = Branch(execute=@n3.else, Condition=IsNamed(self=@n1.AnimInputEvent, InEventName=aie.npc.GiveItem))
   > Set IsGivingItem(IsGivingItem:bool=true) -> (Output_Get:bool)
   Branch(execute=@n4.else, Condition=IsNamed(self=@n1.AnimInputEvent, InEventName=aie.npc.ReceiveItem))
   > Set IsReceivingItem(IsReceivingItem:bool=true) -> (Output_Get:bool)
   Event BlueprintPostEvaluateAnimation() {is=title K2Node_Event}
   > Set IsConversationStarting(IsConversationStarting=false)
   > Set IsConversationEnding(IsConversationEnding=false)
   > Set IsGivingItem(IsGivingItem=false)
   > Set IsReceivingItem(IsReceivingItem=false)
   TryGetPawnOwner()
   Event BlueprintUpdateAnimation()
  ]}
 FullBodyOverride {nodes [
   AnimGraphNode_StateMachine_0 = NPC_Conversation() {is=title AnimGraphNode_StateMachine}
   n1 = "Transform (Modify) Bone - Bone: weapon_l"!(ComponentPose=Local To Component(LocalPose=@AnimGraphNode_StateMachine_0), Alpha=1, Scale="0.001,0.001,0.001") -> (Pose) {scaleMode=BMM_Replace, alphaBoolBlend {blendOption=Linear}, boneToModify {boneName=weapon_l}, scale=<0.001 0.001 0.001>}
   n2 = "Transform (Modify) Bone - Bone: weapon_r"!(ComponentPose=@n1, Alpha=1, Scale="0.001,0.001,0.001") -> (Pose) {scaleMode=BMM_Replace, alphaBoolBlend {blendOption=Linear}, boneToModify {boneName=weapon_r}, scale=<0.001 0.001 0.001>}
   Output Pose(Result=Component To Local(ComponentPose=@n2)) {layerGroup=PlayerOverride}
   Save cached pose 'PassThroughFullBodyPose'(Pose=Input Pose() {name=PassThroughFullBodyPose}) {cacheName=PassThroughFullBodyPose}
   n3 = "Blend Poses (EFortCustomGender)"!(
    BlendPose_0:Pose Link Structure=Sequence Player 'ConversationNPCBasePose'() {sequence=&/ConversationNPCBasePose}
    BlendPose_1:Pose Link Structure=Sequence Player 'ConversationNPCBasePose_CMF'() {bLoopAnimation=false, sequence=&/ConversationNPCBasePose_CMF}
    BlendTime_0:float=0.1
    BlendTime_1:float=0.1
    ActiveEnumValue:Byte=Get Gender() -> (Gender:EFortCustomGender Enum=Invalid)
   ) -> (Pose:Pose Link Structure) {
    is=title AnimGraphNode_BlendListByEnum
    blendType=Linear
    blendPose [{linkId=-1, sourceLinkId=-1}, {linkId=-1, sourceLinkId=-1}]
    blendTime [0.1,0.1]
   }
   n4 = "Sequence Player (additive) 'Idle_Noise'"() {sequence=/Game/Animation/Game/MainPlayer/Locomotion/Medium/Male/Idle/Idle_Noise}
   n5 = Get AdjustedAim() -> (AdjustedAim:Fort Anim Input Adjusted Aim Structure)
   Save cached pose 'DefaultStatePose'(Pose=AimOffset Player 'M_NoWep_Relaxed_AO'!(
     X:float=Break Fort Anim Input Adjusted Aim!(FortAnimInput_AdjustedAim:"Fort Anim Input Adjusted Aim Structure (by ref)"=@n5) -> (ResultingYaw:float=~, ResultingPitch:float=~, ZeroOutPitchWeightCurveName:Name) {is=title K2Node_BreakStruct}
     Y:float=0
     BasePose:Pose Link Structure=Apply Additive(Base=@n3, Additive=@n4) {alphaBoolBlend {blendOption=Linear}}
     Alpha:float=1
    ) -> (Pose:Pose Link Structure) {blendSpace=/Game/Animation/Game/MainPlayer/Poses/Medium/Male/AO/M_NoWep_Relaxed_AO, alphaBoolBlend {blendOption=Linear}}) {cacheName=DefaultStatePose}
  ]}
 NPC_Conversation {
  entry=Default
  path=FullBodyOverride.AnimGraphNode_StateMachine_0.NPC_Conversation
  ids=id
  nodes [
   AnimStateNode_0 = Default() {is=title AnimStateNode}
   AnimStateNode_1 = ConversationStart() {is=title AnimStateNode}
   AnimStateNode_2 = ConversationLoop() {is=title AnimStateNode}
   AnimStateNode_3 = ConversationFinish() {is=title AnimStateNode}
   AnimStateNode_4 = ConversationGiveItem() {is=title AnimStateNode}
   AnimStateNode_5 = ConversationReceiveItem() {is=title AnimStateNode}
   AnimStateTransitionNode_0 = AnimStateTransitionNode()
   AnimStateTransitionNode_1 = AnimStateTransitionNode() {bAutomaticRuleBasedOnSequencePlayerInState=true}
   AnimStateTransitionNode_2 = AnimStateTransitionNode()
   AnimStateTransitionNode_3 = AnimStateTransitionNode() {bAutomaticRuleBasedOnSequencePlayerInState=true}
   AnimStateTransitionNode_4 = AnimStateTransitionNode()
   AnimStateTransitionNode_5 = AnimStateTransitionNode()
   AnimStateTransitionNode_6 = AnimStateTransitionNode()
   AnimStateTransitionNode_7 = AnimStateTransitionNode()
  ]
  links [
   {
    dst=Default
    rule=FullBodyOverride.AnimGraphNode_StateMachine_0.NPC_Conversation.AnimStateTransitionNode_3.Transition
    src=ConversationFinish
    via=AnimStateTransitionNode_3
   }
   {
    dst=ConversationFinish
    rule=FullBodyOverride.AnimGraphNode_StateMachine_0.NPC_Conversation.AnimStateTransitionNode_6.Transition
    src=ConversationGiveItem
    via=AnimStateTransitionNode_6
   }
   {
    dst=ConversationFinish
    rule=FullBodyOverride.AnimGraphNode_StateMachine_0.NPC_Conversation.AnimStateTransitionNode_2.Transition
    src=ConversationLoop
    via=AnimStateTransitionNode_2
   }
   {
    dst=ConversationGiveItem
    rule=FullBodyOverride.AnimGraphNode_StateMachine_0.NPC_Conversation.AnimStateTransitionNode_4.Transition
    src=ConversationLoop
    via=AnimStateTransitionNode_4
   }
   {
    dst=ConversationReceiveItem
    rule=FullBodyOverride.AnimGraphNode_StateMachine_0.NPC_Conversation.AnimStateTransitionNode_5.Transition
    src=ConversationLoop
    via=AnimStateTransitionNode_5
   }
   {
    dst=ConversationFinish
    rule=FullBodyOverride.AnimGraphNode_StateMachine_0.NPC_Conversation.AnimStateTransitionNode_7.Transition
    src=ConversationReceiveItem
    via=AnimStateTransitionNode_7
   }
   {
    dst=ConversationLoop
    rule=FullBodyOverride.AnimGraphNode_StateMachine_0.NPC_Conversation.AnimStateTransitionNode_1.Transition
    src=ConversationStart
    via=AnimStateTransitionNode_1
   }
   {
    dst=ConversationStart
    rule=FullBodyOverride.AnimGraphNode_StateMachine_0.NPC_Conversation.AnimStateTransitionNode_0.Transition
    src=Default
    via=AnimStateTransitionNode_0
   }
  ]
 }
 Default {
  path=FullBodyOverride.AnimGraphNode_StateMachine_0.NPC_Conversation.AnimStateNode_0.Default
  nodes [
   AnimGraphNode_StateMachine_0 = Default() {is=title AnimGraphNode_StateMachine}
   Output Animation Pose(Result=@AnimGraphNode_StateMachine_0)
  ]
 }
 Default {
  entry=Idle
  path=FullBodyOverride.AnimGraphNode_StateMachine_0.NPC_Conversation.AnimStateNode_0.Default.AnimGraphNode_StateMachine_0.Default
  ids=id
  nodes [
   AnimStateNode_0 = Turns() {
    is=title AnimStateNode
    stateEntered {guid=#1}
    stateFullyBlended {guid=#2}
    stateLeft {guid=#3}
   }
   AnimStateNode_2 = Idle()
   AnimStateTransitionNode_0 = AnimStateTransitionNode() {crossfadeDuration=0}
   AnimStateTransitionNode_1 = AnimStateTransitionNode() {crossfadeDuration=0}
   AnimStateTransitionNode_2 = AnimStateTransitionNode() {bAutomaticRuleBasedOnSequencePlayerInState=true, crossfadeDuration=0}
  ]
  links [
   {
    dst=Turns
    rule=FullBodyOverride.AnimGraphNode_StateMachine_0.NPC_Conversation.AnimStateNode_0.Default.AnimGraphNode_StateMachine_0.Default.AnimStateTransitionNode_0.AnimationTransitionGraph_0
    src=Idle
    via=AnimStateTransitionNode_0
   }
   {
    dst=Idle
    rule=FullBodyOverride.AnimGraphNode_StateMachine_0.NPC_Conversation.AnimStateNode_0.Default.AnimGraphNode_StateMachine_0.Default.AnimStateTransitionNode_1.AnimationTransitionGraph_0
    src=Turns
    via=AnimStateTransitionNode_1
   }
   {
    dst=Idle
    rule=FullBodyOverride.AnimGraphNode_StateMachine_0.NPC_Conversation.AnimStateNode_0.Default.AnimGraphNode_StateMachine_0.Default.AnimStateTransitionNode_2.AnimationTransitionGraph_0
    src=Turns
    via=AnimStateTransitionNode_2
   }
  ]
 }
 Turns {
  path=FullBodyOverride.AnimGraphNode_StateMachine_0.NPC_Conversation.AnimStateNode_0.Default.AnimGraphNode_StateMachine_0.Default.AnimStateNode_0.Turns
  nodes [
   n1 = "Sequence Player (additive) 'Idle_Turn_Right_Additive'"() {bLoopAnimation=false, sequence=/Game/Animation/Game/MainPlayer/Locomotion/Small/idle/Idle_Turn_Right_Additive}
   n2 = "Sequence Player (additive) 'Idle_Turn_Left_Additive'"() {bLoopAnimation=false, sequence=/Game/Animation/Game/MainPlayer/Locomotion/Small/idle/Idle_Turn_Left_Additive}
   n3 = Break Fort Anim Input Turn in Place!(FortAnimInput_TurnInPlace:"Fort Anim Input Turn In Place Structure (by ref)"=Get TurnInPlaceInput() -> (TurnInPlaceInput:Fort Anim Input Turn In Place Structure)) -> (
    RootYawOffset:float=~
    TurnAnimPlayRate:float=~
    bWantsToTurn:bool=~
    bExitTurn:bool=~
    bUseTurnA:bool=~
    TurnACurveAlpha:float=~
    TurnBCurveAlpha:float=~
    bWantsToTurnInVehicle:bool=~
    bWantsToTurnAgain:bool=~
    bTurningLeft:bool=~
    ShufflePlayRate:float=~
    CrouchShufflePlayRate:float=~
    CrouchShuffleStartPosition:float=~
    CrouchShuffleSpeedScaling:float=~
    bTransition_LowerBody_Shuffle:bool=~
    bTransition_LowerBody_ShuffleADS:bool=~
    bPlayOutro:bool=~
   ) {is=title K2Node_BreakStruct}
   Output Animation Pose(Result=Apply Additive(
     Base=Use cached pose 'DefaultStatePose'()
     Additive=Blend Poses by bool(BlendPose_0=@n2, BlendPose_1=@n1, bActiveValue=@n3.bTurningLeft) {blendType=Linear}
    ) {alphaBoolBlend {blendOption=Linear}})
  ]
 }
 Idle {
  path=FullBodyOverride.AnimGraphNode_StateMachine_0.NPC_Conversation.AnimStateNode_0.Default.AnimGraphNode_StateMachine_0.Default.AnimStateNode_2.Idle
  nodes [Output Animation Pose(Result=Use cached pose 'DefaultStatePose'())]
 }
 AnimationTransitionGraph_0 {
  path=FullBodyOverride.AnimGraphNode_StateMachine_0.NPC_Conversation.AnimStateNode_0.Default.AnimGraphNode_StateMachine_0.Default.AnimStateTransitionNode_0.AnimationTransitionGraph_0
  nodes [Result(bCanEnterTransition=Get bLocomotion_Idles_to_Turns() -> (bLocomotion_Idles_to_Turns:bool))]
 }
 AnimationTransitionGraph_0 {
  path=FullBodyOverride.AnimGraphNode_StateMachine_0.NPC_Conversation.AnimStateNode_0.Default.AnimGraphNode_StateMachine_0.Default.AnimStateTransitionNode_1.AnimationTransitionGraph_0
  nodes [Result(bCanEnterTransition=Break Fort Anim Input Turn in Place(FortAnimInput_TurnInPlace=Get TurnInPlaceInput()) -> (
     RootYawOffset=~
     TurnAnimPlayRate=~
     bWantsToTurn=~
     bExitTurn=~
     bUseTurnA=~
     TurnACurveAlpha=~
     TurnBCurveAlpha=~
     bWantsToTurnInVehicle=~
     bWantsToTurnAgain=~
     bTurningLeft=~
     ShufflePlayRate=~
     CrouchShufflePlayRate=~
     CrouchShuffleStartPosition=~
     CrouchShuffleSpeedScaling=~
     bTransition_LowerBody_Shuffle=~
     bTransition_LowerBody_ShuffleADS=~
     bPlayOutro=~
    ).bWantsToTurnAgain)]
 }
 AnimationTransitionGraph_0 {
  path=FullBodyOverride.AnimGraphNode_StateMachine_0.NPC_Conversation.AnimStateNode_0.Default.AnimGraphNode_StateMachine_0.Default.AnimStateTransitionNode_2.AnimationTransitionGraph_0
  nodes [Result()]
 }
 ConversationStart {
  path=FullBodyOverride.AnimGraphNode_StateMachine_0.NPC_Conversation.AnimStateNode_1.ConversationStart
  nodes [
   n1 = "Blend Poses (EFortCustomGender)"(
    BlendPose_0=Sequence Player 'ConversationStart'() {bLoopAnimation=false, sequence=&/ConversationStart}
    BlendPose_1=Sequence Player 'ConversationStart_CMF'() {bLoopAnimation=false, sequence=&/ConversationStart_CMF}
    BlendTime_0=0.1
    BlendTime_1=0.1
    ActiveEnumValue=Get Gender() -> (Gender=Invalid)
   ) {blendType=Linear, blendPose [{linkId=-1, sourceLinkId=-1}, {linkId=-1, sourceLinkId=-1}], blendTime [0.1,0.1]}
   Output Animation Pose(Result=@n1)
  ]
 }
 ConversationLoop {
  path=FullBodyOverride.AnimGraphNode_StateMachine_0.NPC_Conversation.AnimStateNode_2.ConversationLoop
  nodes [
   n1 = "Blend Poses (EFortCustomGender)"(
    BlendPose_0=Sequence Player 'ConversationLoop'() {sequence=&/ConversationLoop}
    BlendPose_1=Sequence Player 'ConversationLoop_CMF'() {bLoopAnimation=false, sequence=&/ConversationLoop_CMF}
    BlendTime_0=0.1
    BlendTime_1=0.1
    ActiveEnumValue=Get Gender() -> (Gender=Invalid)
   ) {blendType=Linear, blendPose [{linkId=-1, sourceLinkId=-1}, {linkId=-1, sourceLinkId=-1}], blendTime [0.1,0.1]}
   Output Animation Pose(Result=@n1)
  ]
 }
 ConversationFinish {
  path=FullBodyOverride.AnimGraphNode_StateMachine_0.NPC_Conversation.AnimStateNode_3.ConversationFinish
  nodes [
   n1 = "Blend Poses (EFortCustomGender)"(
    BlendPose_0=Sequence Player 'ConversationFinish'() {bLoopAnimation=false, sequence=&/ConversationFinish}
    BlendPose_1=Sequence Player 'ConversationFinish_CMF'() {bLoopAnimation=false, sequence=&/ConversationFinish_CMF}
    BlendTime_0=0.1
    BlendTime_1=0.1
    ActiveEnumValue=Get Gender() -> (Gender=Invalid)
   ) {blendType=Linear, blendPose [{linkId=-1, sourceLinkId=-1}, {linkId=-1, sourceLinkId=-1}], blendTime [0.1,0.1]}
   Output Animation Pose(Result=@n1)
  ]
 }
 ConversationGiveItem {
  path=FullBodyOverride.AnimGraphNode_StateMachine_0.NPC_Conversation.AnimStateNode_4.ConversationGiveItem
  nodes [
   n1 = "Blend Poses (EFortCustomGender)"(
    BlendPose_0=Sequence Player 'ConversationGiveItem'() {bLoopAnimation=false, sequence=&/ConversationGiveItem}
    BlendPose_1=Sequence Player 'ConversationGiveItem_CMF'() {bLoopAnimation=false, sequence=&/ConversationGiveItem_CMF}
    BlendTime_0=0.1
    BlendTime_1=0.1
    ActiveEnumValue=Get Gender() -> (Gender=Invalid)
   ) {blendType=Linear, blendPose [{linkId=-1, sourceLinkId=-1}, {linkId=-1, sourceLinkId=-1}], blendTime [0.1,0.1]}
   Output Animation Pose(Result=@n1)
  ]
 }
 ConversationReceiveItem {
  path=FullBodyOverride.AnimGraphNode_StateMachine_0.NPC_Conversation.AnimStateNode_5.ConversationReceiveItem
  nodes [
   n1 = "Blend Poses (EFortCustomGender)"(
    BlendPose_0=Sequence Player 'ConversationReceiveItem'() {bLoopAnimation=false, sequence=&/ConversationReceiveItem}
    BlendPose_1=Sequence Player 'ConversationReceiveItem_CMF'() {bLoopAnimation=false, sequence=&/ConversationReceiveItem_CMF}
    BlendTime_0=0.1
    BlendTime_1=0.1
    ActiveEnumValue=Get Gender() -> (Gender=Invalid)
   ) {blendType=Linear, blendPose [{linkId=-1, sourceLinkId=-1}, {linkId=-1, sourceLinkId=-1}], blendTime [0.1,0.1]}
   Output Animation Pose(Result=@n1)
  ]
 }
 Transition {
  path=FullBodyOverride.AnimGraphNode_StateMachine_0.NPC_Conversation.AnimStateTransitionNode_0.Transition
  nodes [Result(bCanEnterTransition=Get IsConversationStarting() -> (IsConversationStarting:bool))]
 }
 Transition {path=FullBodyOverride.AnimGraphNode_StateMachine_0.NPC_Conversation.AnimStateTransitionNode_1.Transition, nodes [Result()]}
 Transition {
  path=FullBodyOverride.AnimGraphNode_StateMachine_0.NPC_Conversation.AnimStateTransitionNode_2.Transition
  nodes [Result(bCanEnterTransition=Get IsConversationEnding() -> (IsConversationEnding:bool))]
 }
 Transition {path=FullBodyOverride.AnimGraphNode_StateMachine_0.NPC_Conversation.AnimStateTransitionNode_3.Transition, nodes [Result()]}
 Transition {
  path=FullBodyOverride.AnimGraphNode_StateMachine_0.NPC_Conversation.AnimStateTransitionNode_4.Transition
  nodes [Result(bCanEnterTransition=Get IsGivingItem() -> (IsGivingItem:bool))]
 }
 Transition {
  path=FullBodyOverride.AnimGraphNode_StateMachine_0.NPC_Conversation.AnimStateTransitionNode_5.Transition
  nodes [Result(bCanEnterTransition=Get IsReceivingItem() -> (IsReceivingItem:bool))]
 }
 Transition {
  path=FullBodyOverride.AnimGraphNode_StateMachine_0.NPC_Conversation.AnimStateTransitionNode_6.Transition
  nodes [Result(bCanEnterTransition=Get IsConversationEnding())]
 }
 Transition {
  path=FullBodyOverride.AnimGraphNode_StateMachine_0.NPC_Conversation.AnimStateTransitionNode_7.Transition
  nodes [Result(bCanEnterTransition=Get IsConversationEnding())]
 }
 PreLegIK {nodes [
   Input Pose()
   Output Pose() {layerGroup=PlayerOverride}
  ]}
 UpperBody {nodes [
   Input Pose() {name=InPoseUpperBody}
   Output Pose() {layerGroup=PlayerOverride}
  ]}
]
variables [
 IsConversationEnding {type=boolean}
 IsConversationStarting {type=boolean}
 IsGivingItem {type=boolean}
 IsReceivingItem {type=boolean}
]
