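; BP_CalendarDriven_PathProvider gates Living World point providers on a calendar event, apparently for NPC paths.
; Once playlist data is ready, it checks for a debug bypass, an empty event name, or an active event in the configured abduction phase (-1 matches any).
; GetAbductionPhase splits event progress into three phases for that check.
; Valid providers in the PointProvider array are enabled through the deprecated EnablePointProvider interface call.
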(event EventBeginPlay
  (bind _calendareventname (Variables|Default|GetCalendarEventName))
  (Class|FortAsyncActionCurrentPlaylistReady|PlaylistDataReadyAsync
    (:OnReady
      (if (Variables|Default|GetDebugBypassCalendarEvent)
        _
        (elif (Utilities|String|IsEmpty _calendareventname)
          _
          (else
            (bind (_biseventactive _timesincebegin _timeuntilend _timespanratio) (CalendarEvents|GetCalendarEventInformation _calendareventname))
            (if (and _biseventactive (or (== (CallFunction|GetAbductionPhase _timespanratio) (Variables|Default|GetAbductionPhase)) (== (Variables|Default|GetAbductionPhase) -1)))
              _))))
      (for _array_element (Variables|Default|GetPointProvider)
        (Utilities|IsValid _array_element
          (:Is%20Valid
            (Class|FortAthenaLivingWorldPointProviderInterface|EnablePointProvider_DEPRECATED _array_element)))))))

(event Collision|EventActorBeginOverlap
  (params (OtherActor :type Actor%20Object%20Reference)))

(event EventTick
  (params (DeltaSeconds :type Float%20%28single-precision%29)))

(fn GetAbductionPhase
  (params (Progress :type Float%20%28double-precision%29))
  (if (< Progress 0.3)
    (LocalVariables|SetAbduction%20Phase 0)
    (elif (< Progress 0.6)
      (LocalVariables|SetAbduction%20Phase 1)
      (elif (>= Progress 0.6)
        (LocalVariables|SetAbduction%20Phase 2)
        (else
          (Development|PrintString "NoAbductionPhaseSet")))))
  (return (LocalVariables|GetAbduction%20Phase)))

(fn ConstructionScript)
