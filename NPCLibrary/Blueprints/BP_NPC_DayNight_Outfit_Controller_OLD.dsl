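; NPC controller that swaps a spawned NPC's cosmetic outfit between day and night versions.
; On authority, once playlist data is ready, it enables a living world point provider and binds to its actor spawned result.
; The spawned pawn gets a repeating timer that activates the night cosmetic during Evening or Night and deactivates it otherwise, with a flag preventing redundant swaps.
; The timer is cleared when the pawn dies.
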(event EventBeginPlay
  (bind _returnvalue (Game|GetGameState))
  (bind _returnvalue_2 (Actor|GetComponentbyClass _returnvalue "/Script/LagerRuntime.FortAthenaLivingWorldManager"))
  (bind _pathprovider (Variables|Default|GetPathProvider))
  ;[1 1 1] Begin Play
  (Class|FortAsyncActionCurrentPlaylistReady|PlaylistDataReadyAsync
    (:OnReady
      (Networking|SwitchHasAuthority
        (:Authority
          (Utilities|IsValid _returnvalue_2
            (:Is%20Valid
              (Variables|Default|SetLivingWorldManager _returnvalue_2)
              (bind _returnvalue_3 (GUID|GenerateGuidfromObject (Variables|Default|GetPathProvider)))
              (Variables|Default|SetSwapId _returnvalue_3)
              (Class|FortAthenaLivingWorldPointProviderInterface|EnablePointProvider_DEPRECATED _pathprovider)
              (EventDispatchers|BindEventtoOnActorSpawnedResult _pathprovider (EventDispatchers|CreateEvent :!event Custom|ActorSpawned)))))))))

(event Custom|ActorSpawned
  (params (SpawnedActor :type Actor%20Object%20Reference) (bResult :type Boolean))
  ;[1 1 1] Set Up Timer After Actor Spawned
  (bind _asfort_player_pawn (Utilities|Casting|CastToFortPlayerPawn SpawnedActor))
  (bind _output_get (Variables|Default|SetFortPlayerPawnRef _asfort_player_pawn))
  (EventDispatchers|BindEventtoOnDied _output_get (EventDispatchers|CreateEvent :!event Custom|OnDied))
  (bind _returnvalue (Utilities|Time|SetTimerbyEvent (EventDispatchers|CreateEvent :!event Custom|CheckTimeOfDay) (Variables|Default|GetCheckTImeInterval) true)) ; Check the time of day periodically
  (Variables|Default|SetTrySpawnTimerHandle _returnvalue))

(event Custom|CheckTimeOfDay
  ;[1 1 1] Check Time of Day and Set Outfit
  (if (or (Utilities|Enum|Equal(Enum) (TimeOfDay|GetCurrentDayPhase :!class /Script/FortniteGame.FortKismetLibrary) "Evening") (Utilities|Enum|Equal(Enum) (TimeOfDay|GetCurrentDayPhase :!class /Script/FortniteGame.FortKismetLibrary) "Night"))
    (if (not (Variables|Default|GetbNightOverrideActive))
      (Class|FortPlayerPawn|SwapPawnCosmeticItemandVariant (Variables|Default|GetFortPlayerPawnRef) (Variables|Default|GetSwapId) (Variables|Default|GetSwapDataNight) false "Active" false) ; Change the outfit out to the night version
      (Variables|Default|SetbNightOverrideActive true))
    (elif (Variables|Default|GetbNightOverrideActive)
      (Class|FortPlayerPawn|SwapPawnCosmeticItemandVariant (Variables|Default|GetFortPlayerPawnRef) (Variables|Default|GetSwapId) (Variables|Default|GetSwapDataNight) false "Inactive" false)
      (Variables|Default|SetbNightOverrideActive false))))

(event Custom|OnDied
  (params (DamagedActor :type Actor%20Object%20Reference) (Damage :type Float%20%28single-precision%29) (InstigatedBy :type Controller%20Object%20Reference) (DamageCauser :type Actor%20Object%20Reference) (HitLocation :type Vector) (FHitComponent :type Primitive%20Component%20Object%20Reference) (BoneName :type Name) (Momentum :type Vector))
  ;[1 1 1] Clear Day/Night Check Timer On Death
  (Utilities|Time|ClearandInvalidateTimerbyHandle (Variables|Default|GetTrySpawnTimerHandle)))

(fn ConstructionScript)
